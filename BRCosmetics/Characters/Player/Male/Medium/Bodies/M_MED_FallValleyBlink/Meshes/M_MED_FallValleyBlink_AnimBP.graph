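# Post-process style cosmetic AnimBP for the FallValleyBlink body that adds secondary motion to its gear on the pose copied from the parent mesh.
# A Control Rig feeds spring AnimDynamics and additive bone offsets for pouches, belt, harness, holsters and knife, then constraints on holsters.
# Updates derive offset alphas from thigh and clavicle deltas, adjusted for crouch and parachute.
# Events toggle a material hide parameter by matching equipped pickaxe and skydiving state.
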AnimBlueprint /BRCosmetics/Characters/Player/Male/Medium/Bodies/M_MED_FallValleyBlink/Meshes/M_MED_FallValleyBlink_AnimBP
target_skeleton=&/M_MED_FallValleyBlink_Skeleton
graphs [
 AnimGraph {nodes [
   Control Rig() {
    alphaBoolBlend {blendOption=Linear}
    controlRigAssetReference {blueprintRigClass=&/M_MED_FallValleyBlink_Control_Rig_Twists.M_MED_FallValleyBlink_Control_Rig_Twists_C}
   }
   n1 = Control Rig(Source=Copy Pose From Mesh() {bUseAttachedParent=true}) {
    alphaBoolBlend {blendOption=Linear}
    controlRigAssetReference {blueprintRigClass=&/M_MED_FallValleyBlink_Control_Rig.M_MED_FallValleyBlink_Control_Rig_C}
   }
   n2 = Anim Dynamics - Bone: dyn_pouch_bck_l!(ComponentPose=Local To Component(LocalPose=@n1), Alpha=1, GravityOverride:Vector="0,0,-980", ExternalForce="0,0,0") -> (Pose) {
    angularDampingOverride=0.999
    angularSpringConstant=300
    bAngularSpring=true
    bOverrideAngularDamping=true
    bUseGravityOverride=true
    linearDampingOverride=0.999
    linearSpringConstant=300
    simulationSpace=RootRelative
    alphaBoolBlend {blendOption=Linear}
    boundBone {boneName=dyn_pouch_bck_l}
    componentAppliedLinearAccClamp=<5000 5000 5000>
    componentLinearAccScale=<0.5 0.5 0.5>
    componentLinearVelScale=<0.5 0.5 0.5>
    physicsBodyDefinitions [{
      boundBone {boneName=dyn_pouch_bck_l}
      boxExtents=<20 20 20>
      constraintSetup {angularLimitsMax=<0 0 20>, angularLimitsMin=<0 0 -2>, angularTarget=<0 0 0>}
      localJointOffset=<1 1 0>
     }]
   }
   n3 = Anim Dynamics - Bone: dyn_pouch_bck!(ComponentPose=@n2, Alpha=1, GravityOverride:Vector="0,0,-300", ExternalForce="0,0,0") -> (Pose) {
    angularDampingOverride=0.999
    angularSpringConstant=200
    bAngularSpring=true
    bOverrideAngularDamping=true
    bUseGravityOverride=true
    linearDampingOverride=0.999
    linearSpringConstant=300
    simulationSpace=RootRelative
    alphaBoolBlend {blendOption=Linear}
    boundBone {boneName=dyn_pouch_bck}
    componentAppliedLinearAccClamp=<5000 5000 5000>
    componentLinearAccScale=<0.5 0.5 0.5>
    componentLinearVelScale=<0.5 0.5 0.5>
    physicsBodyDefinitions [{
      boundBone {boneName=dyn_pouch_bck}
      boxExtents=<5 5 5>
      constraintSetup {angularLimitsMax=<0 0 10>, angularLimitsMin=<0 0 0>, angularTarget=<0 0 0>}
      localJointOffset=<1 1 0>
     }]
   }
   n4 = Anim Dynamics - Bone: dyn_pouch_bck_r!(ComponentPose=@n3, Alpha=1, GravityOverride:Vector="0,0,-980", ExternalForce="0,0,0") -> (Pose) {
    angularDampingOverride=0.999
    angularSpringConstant=200
    bAngularSpring=true
    bOverrideAngularDamping=true
    bUseGravityOverride=true
    linearDampingOverride=0.999
    linearSpringConstant=300
    simulationSpace=RootRelative
    alphaBoolBlend {blendOption=Linear}
    boundBone {boneName=dyn_pouch_bck_r}
    componentAppliedLinearAccClamp=<5000 5000 5000>
    componentLinearAccScale=<0.5 0.5 0.5>
    componentLinearVelScale=<0.5 0.5 0.5>
    physicsBodyDefinitions [{
      boundBone {boneName=dyn_pouch_bck_r}
      boxExtents=<20 20 20>
      constraintSetup {angularLimitsMax=<0 0 20>, angularLimitsMin=<0 0 -1>, angularTarget=<0 0 0>}
      localJointOffset=<1 1 0>
     }]
   }
   n5 = "Transform (Modify) Bone - Bone: dyn_pouch_fr_r"(
    ComponentPose=@n4
    Alpha=Get R_pouch_front() -> (R_pouch_front:double)
    Translation=Get R_pouch_front_Translation() -> (R_pouch_front_Translation:Vector)
    Rotation=Get R_pouch_front_Rotation() -> (R_pouch_front_Rotation:Rotator)
    Scale="1,1,1"
   ) {
    rotationMode=BMM_Additive
    rotationSpace=BCS_BoneSpace
    translationMode=BMM_Additive
    translationSpace=BCS_BoneSpace
    alphaBoolBlend {blendOption=Linear}
    alphaScaleBiasClamp {bInterpResult=true, interpSpeedIncreasing=20}
    boneToModify {boneName=dyn_pouch_fr_r}
    rotation {pitch=100, yaw=70}
   }
   n6 = "Transform (Modify) Bone - Bone: dyn_pouch_fr_l"(ComponentPose=@n5, Alpha=Get l_pouch_front() -> (l_pouch_front:double), Translation="-2.5,-4,-8", Rotation="18,-8,20", Scale="1,1,1") {
    rotationMode=BMM_Additive
    rotationSpace=BCS_BoneSpace
    translationMode=BMM_Additive
    translationSpace=BCS_BoneSpace
    alphaBoolBlend {blendOption=Linear}
    alphaScaleBiasClamp {bInterpResult=true, interpSpeedDecreasing=15, interpSpeedIncreasing=20}
    boneToModify {boneName=dyn_pouch_fr_l}
    rotation=<rot 18 -8 20>
    translation=<-2.5 -4 -8>
   }
   n7 = "Transform (Modify) Bone - Bone: dyn_pouch_leg_l"(ComponentPose=@n6, Alpha=Get L_Leg_pouch_crouch() -> (L_Leg_pouch_crouch:double), Translation="-14,-2,-8", Rotation="0,0,-35", Scale="1,1,1") {
    rotationMode=BMM_Additive
    rotationSpace=BCS_ParentBoneSpace
    translationMode=BMM_Additive
    translationSpace=BCS_BoneSpace
    alphaBoolBlend {blendOption=Linear}
    boneToModify {boneName=dyn_pouch_leg_l}
    rotation {roll=-35}
    translation=<-14 -2 -8>
   }
   n8 = "Transform (Modify) Bone - Bone: harness_l"(
    ComponentPose=@n7
    Alpha=Get harness_l() -> (harness_l:double)
    Translation=Get Harness_translate_l() -> (Harness_translate_l:Vector)
    Rotation=Get Harness_rotator_L() -> (Harness_rotator_L:Rotator)
    Scale="1,1,1"
   ) {
    rotationMode=BMM_Additive
    rotationSpace=BCS_BoneSpace
    translationMode=BMM_Additive
    translationSpace=BCS_BoneSpace
    alphaBoolBlend {blendOption=Linear}
    boneToModify {boneName=harness_l}
    rotation {pitch=20}
    translation {x=9, z=1}
   }
   n9 = "Transform (Modify) Bone - Bone: harness_r"(
    ComponentPose=@n8
    Alpha=Get harness_r() -> (harness_r:double)
    Translation=Get Harness_translate_R() -> (Harness_translate_R:Vector)
    Rotation=Get Harness_rotator_R() -> (Harness_rotator_R:Rotator)
    Scale="1,1,1"
   ) {
    rotationMode=BMM_Additive
    rotationSpace=BCS_BoneSpace
    translationMode=BMM_Additive
    translationSpace=BCS_BoneSpace
    alphaBoolBlend {blendOption=Linear}
    boneToModify {boneName=harness_r}
    rotation {pitch=20}
    translation {x=-9, z=-1}
   }
   n10 = Get l_pouch_front()
   n11 = "Transform (Modify) Bone - Bone: dyn_belt_fr_l"(ComponentPose=@n9, Alpha=@n10, Translation="-10,1,0", Rotation="0,0,0", Scale="1,1,1") {
    rotationMode=BMM_Additive
    rotationSpace=BCS_BoneSpace
    translationMode=BMM_Additive
    translationSpace=BCS_BoneSpace
    alphaBoolBlend {blendOption=Linear}
    alphaScaleBiasClamp {bInterpResult=true, interpSpeedDecreasing=15, interpSpeedIncreasing=20}
    boneToModify {boneName=dyn_belt_fr_l}
    translation=<-10 1>
   }
   n12 = "Transform (Modify) Bone - Bone: dyn_belt_fr"(ComponentPose=@n11, Alpha=@n10, Translation="-5,1,-2", Rotation="15,0,13", Scale="1,1,1") {
    rotationMode=BMM_Additive
    rotationSpace=BCS_BoneSpace
    translationMode=BMM_Additive
    translationSpace=BCS_BoneSpace
    alphaBoolBlend {blendOption=Linear}
    alphaScaleBiasClamp {bInterpResult=true, interpSpeedDecreasing=15, interpSpeedIncreasing=20}
    boneToModify {boneName=dyn_belt_fr}
    rotation {pitch=15, roll=13}
    translation=<-5 1 -2>
   }
   n13 = "Transform (Modify) Bone - Bone: dyn_belt_fr_r"(ComponentPose=@n12, Alpha=@n10, Translation="-1.5,-0.5,-2", Rotation="17,0,15", Scale="1,1,1") {
    rotationMode=BMM_Additive
    rotationSpace=BCS_BoneSpace
    translationMode=BMM_Additive
    translationSpace=BCS_BoneSpace
    alphaBoolBlend {blendOption=Linear}
    alphaScaleBiasClamp {bInterpResult=true, interpSpeedDecreasing=15, interpSpeedIncreasing=20}
    boneToModify {boneName=dyn_belt_fr_r}
    rotation {pitch=17, roll=15}
    translation=<-1.5 -0.5 -2>
   }
   n14 = Anim Dynamics - Bone: dyn_holster_r!(ComponentPose=@n13, Alpha=1, GravityOverride:Vector="0,0,-980", ExternalForce="0,0,0") -> (Pose) {
    angularDampingOverride=0.99
    angularSpringConstant=300
    bAngularSpring=true
    bOverrideAngularDamping=true
    bOverrideLinearDamping=true
    bUseGravityOverride=true
    bUseSphericalLimits=true
    linearDampingOverride=0.99
    linearSpringConstant=200
    simulationSpace=RootRelative
    alphaBoolBlend {blendOption=Linear}
    boundBone {boneName=dyn_holster_r}
    componentAppliedLinearAccClamp=<5000 5000 5000>
    componentLinearAccScale=<0.5 0.5 0.5>
    componentLinearVelScale=<0.5 0.5 0.5>
    physicsBodyDefinitions [{
      collisionType=CustomSphere
      sphereCollisionRadius=2
      boundBone {boneName=dyn_holster_r}
      boxExtents=<20 20 20>
      constraintSetup {angularLimitsMax=<0 0 20>, angularLimitsMin=<0 0 0>, angularTarget=<0 0 0>}
      localJointOffset=<1 0 0>
     }]
    sphericalLimits [
     {limitRadius=14, drivingBone {boneName=spine_03}, sphereLocalOffset=<0 0 0>}
     {limitRadius=15, drivingBone {boneName=spine_04}, sphereLocalOffset=<0 0 0>}
    ]
   }
   n15 = Anim Dynamics - Bone: dyn_holster_l!(ComponentPose=@n14, Alpha=1, GravityOverride:Vector="0,0,-980", ExternalForce="0,0,0") -> (Pose) {
    angularDampingOverride=0.999
    angularSpringConstant=300
    bAngularSpring=true
    bOverrideAngularDamping=true
    bOverrideLinearDamping=true
    bUseGravityOverride=true
    bUseSphericalLimits=true
    linearDampingOverride=0.99
    linearSpringConstant=200
    simulationSpace=RootRelative
    alphaBoolBlend {blendOption=Linear}
    boundBone {boneName=dyn_holster_l}
    componentAppliedLinearAccClamp=<5000 5000 5000>
    componentLinearAccScale=<0.5 0.5 0.5>
    componentLinearVelScale=<0.5 0.5 0.5>
    physicsBodyDefinitions [{
      collisionType=CustomSphere
      sphereCollisionRadius=2
      boundBone {boneName=dyn_holster_l}
      boxExtents=<20 20 20>
      constraintSetup {angularLimitsMax=<0 0 20>, angularLimitsMin=<0 0 0>, angularTarget=<0 0 0>}
      localJointOffset=<1 0 0>
     }]
    sphericalLimits [
     {limitRadius=13.4, drivingBone {boneName=spine_03}, sphereLocalOffset=<0 0 0>}
     {limitRadius=15, drivingBone {boneName=spine_04}, sphereLocalOffset=<0 0 0>}
     {limitRadius=10, drivingBone {boneName=dyn_pouch_fr_l}, sphereLocalOffset=<2 0 0>}
    ]
   }
   n16 = Anim Dynamics - Bone: dyn_holster_leg_r(ComponentPose=@n15, Alpha=1, ExternalForce="0,0,0") {
    angularDampingOverride=0.99
    angularSpringConstant=600
    bAngularSpring=true
    bOverrideAngularDamping=true
    bOverrideLinearDamping=true
    bUseGravityOverride=true
    linearDampingOverride=0.99
    linearSpringConstant=600
    simulationSpace=RootRelative
    alphaBoolBlend {blendOption=Linear}
    boundBone {boneName=dyn_holster_leg_r}
    componentAppliedLinearAccClamp=<2000 2000 2000>
    componentLinearAccScale {x=0.5}
    componentLinearVelScale {x=0.5}
    physicsBodyDefinitions [{
      boundBone {boneName=dyn_holster_leg_r}
      boxExtents=<15 15 15>
      constraintSetup {twistAxis=AxisZ, angularLimitsMax=<5 0 0>, angularLimitsMin=<-5 0 0>, angularTarget=<0 0 0>}
      localJointOffset=<0 2 0>
     }]
   }
   n17 = Anim Dynamics - Bone: dyn_pouch_fr_l(ComponentPose=@n16, Alpha=1, ExternalForce="0,0,0") {
    bUsePlanarLimit=false
    bUseSphericalLimits=true
    simulationSpace=RootRelative
    alphaBoolBlend {blendOption=Linear}
    boundBone {boneName=dyn_pouch_fr_l}
    physicsBodyDefinitions [{
      collisionType=CustomSphere
      sphereCollisionRadius=2
      boundBone {boneName=dyn_pouch_fr_l}
      boxExtents=<10 10 10>
      constraintSetup {angularLimitsMax=<0 0 0>, angularLimitsMin=<0 0 0>, angularTarget=<0 0 0>}
      localJointOffset=<0 0 0>
     }]
    sphericalLimits [{limitRadius=12, drivingBone {boneName=spine_03}, sphereLocalOffset=<0 0 0>}]
   }
   n18 = Anim Dynamics - Bone: dyn_knife(ComponentPose=@n17, Alpha=1, ExternalForce="0,0,0") {
    angularDampingOverride=0.99
    angularSpringConstant=200
    bAngularSpring=true
    bOverrideAngularDamping=true
    bOverrideLinearDamping=true
    bUseGravityOverride=true
    linearDampingOverride=0.99
    linearSpringConstant=50
    simulationSpace=RootRelative
    alphaBoolBlend {blendOption=Linear}
    boundBone {boneName=dyn_knife}
    componentAppliedLinearAccClamp=<2000 2000 2000>
    componentLinearAccScale=<0.3 0.3 0.3>
    componentLinearVelScale=<0.3 0.3 0.3>
    physicsBodyDefinitions [{
      boundBone {boneName=dyn_knife}
      boxExtents=<30 30 30>
      constraintSetup {angularLimitsMax=<1 5 5>, angularLimitsMin=<-1 -5 -5>, angularTarget=<0 0 0>}
      localJointOffset=<1 0 0>
     }]
   }
   n19 = "Constraint - dyn_holster_l (P)"(ComponentPose=@n18, Alpha=0.6, ConstraintWeights_0=1) {
    alphaBoolBlend {blendOption=Linear}
    boneToModify {boneName=dyn_holster_l}
    constraintSetup [{targetBone {boneName=spine_03}}]
    constraintWeights [1]
   }
   n20 = "Constraint - dyn_holster_r (P)"(ComponentPose=@n19, Alpha=0.6, ConstraintWeights_0=1) {
    alphaBoolBlend {blendOption=Linear}
    boneToModify {boneName=dyn_holster_r}
    constraintSetup [{targetBone {boneName=spine_03}}]
    constraintWeights [1]
   }
   Output Pose(Result=Component To Local(ComponentPose=@n20)) {layerGroup=None}
  ]}
 EventGraph {nodes [
   Event BlueprintUpdateAnimation()
   > Set harness_l(harness_l:double=GetAlphaFromReferenceTransformDelta(
     SkeletalMeshComponent=Get SkelMesh()
     BoneName=clavicle_l
     BaseName=spine_05
     DeltaComponentType=EulerY
     InputRange=(X=17,Y=48)
     TwistAxis="1,0,0"
    )) -> (Output_Get:double)
   > Set harness_r(harness_r:double=GetAlphaFromReferenceTransformDelta(
     SkeletalMeshComponent=Get SkelMesh()
     BoneName=clavicle_r
     BaseName=spine_05
     DeltaComponentType=EulerY
     InputRange=(X=17,Y=48)
     TwistAxis="1,0,0"
    )) -> (Output_Get:double)
   n1 = Get bIsParachuteOpened()
   > Set Harness_translate_l(Harness_translate_l:Vector=SelectVector(A="9.5,2,0", B="8.5,0,0", bPickA=@n1)) -> (Output_Get:Vector)
   > Set Harness_translate_R(Harness_translate_R:Vector=SelectVector(A="-7,-4,-1.5", B="-7,0,0", bPickA=@n1)) -> (Output_Get:Vector)
   > Set Harness_rotator_L(Harness_rotator_L:Rotator=SelectRotator!(A:Rotator="30,0,0", B:Rotator="25,0,0", bPickA:bool=@n1) -> (ReturnValue:Rotator)) -> (Output_Get:Rotator)
   > Set Harness_rotator_R(Harness_rotator_R:Rotator=SelectRotator(A="15,0,0", B="25,0,0", bPickA=@n1)) -> (Output_Get:Rotator)
   > Set R_pouch_front(R_pouch_front:double=GetAlphaFromReferenceTransformDelta(SkeletalMeshComponent=Get SkelMesh(), BoneName=thigh_r, BaseName=pelvis, InputRange=(X=0,Y=-102), TwistAxis="1,0,0")) -> (Output_Get:double)
   n2 = Get bIsCrouching()
   > Set R_pouch_front_Translation(R_pouch_front_Translation:Vector=SelectVector(A="0, 0, 0", B="-1,0,0", bPickA=@n2)) -> (Output_Get:Vector)
   > Set R_pouch_front_Rotation(R_pouch_front_Rotation:Rotator=SelectRotator(A="100,70,0", B="-80,20,10", bPickA=@n2)) -> (Output_Get:Rotator)
   n3 = GetAlphaFromReferenceTransformDelta(SkeletalMeshComponent=Get SkelMesh(), BoneName=thigh_l, BaseName=pelvis, InputRange=(X=-18,Y=-125), TwistAxis="1,0,0")
   > Set l_pouch_front(l_pouch_front:double=@n3) -> (Output_Get:double)
   n4 = Get bIsCrouching()
   > Set L_Leg_pouch_crouch(L_Leg_pouch_crouch:double=float * float(A=@n3, B="To Float (Boolean)"(InBool=@n4)) {is=title K2Node_PromotableOperator}) -> (Output_Get:double)
   Event BlueprintInitializeAnimation()
   > Set SkelMesh(SkelMesh=GetOwnerSkeletalMeshComponent())
   n5 = Event OnCharacterCustomizationLoaded()
   n6 = Get Dynamic Material Instance(execute=[@n5.then, @n12.then], SkeletalMeshComponent=GetPartSkeletalMeshComponent())
   n7 = PickaxeItemDefinitionHasAnyMetaTag(Tags=Make Array("[0]"=(TagName="Cosmetics.MetaTags.FallValleyBlink")), bRequiresEquipped=true)
   > n8 = Branch(Condition=@n7)
   > n9 = Branch(Condition=Get bIsSkyDiving())
   SetScalarParameterValue(execute=[@n10.then, @n11.else, @n8.else, @n9.then], self=@n6, ParameterName=Hide Element 01, Value=0)
   SetScalarParameterValue(execute=[@n11.then, @n9.else], self=@n6, ParameterName=Hide Element 01, Value=1)
   n10 = Event OnBeginSkydiving()
   Event OnEndSkydiving()
   > n11 = Branch(Condition=@n7)
   n12 = Event OnWeaponChanged()
  ]}
]
variables [
 Crouch_Stand_side_L {type=number}
 harness_l {type=number}
 harness_r {type=number}
 Harness_rotator_L {type=Rotator}
 Harness_rotator_R {type=Rotator}
 Harness_translate_l {type=Vector}
 Harness_translate_R {type=Vector}
 L_Leg_pouch_crouch {type=number}
 l_pouch_front {type=number}
 L_Pouch_front_Jump {type=number}
 R_pouch_front {type=number}
 R_pouch_front_Rotation {type=Rotator, default=<rot 100 70 0>}
 R_Pouch_front_Stand {type=number}
 R_pouch_front_Translation {type=Vector}
 Skeletal Mesh Component {type=SkinnedMeshComponent, default=null}
 Skeletal Mesh Component_0 {type=SkinnedMeshComponent, default=null}
 SkelMesh {type=SkeletalMeshComponent}
]
