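# Physics asset for the face accessory mesh of the male FallValleyBlink character cosmetic, covering its side and front hair bones.
# Small spheres sit on the bones of left and right locks and five front strands, alongside one capsule on the head.
# Constraints chain each lock's neighboring bones with modest swing and twist angles, somewhat wider on strands A, B and E.
# Only strand C is constrained to the head; other chain roots have no constraint to it.
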PhysicsAsset /BRCosmetics/Characters/Player/Male/Medium/Bodies/M_MED_FallValleyBlink/Meshes/Parts/M_MED_FallValleyBlink_FaceAcc_Physics
bodies {
 dyn_fr_l_hair_1 [{shapeType=Sphere, radius=1, shapeName=dyn_fr_l_hair_1_sphere, center=<0 0 -0.8989>}]
 dyn_fr_l_hair_2 [{shapeType=Sphere, radius=1.5, shapeName=dyn_fr_l_hair_2_sphere, center=<-0.5518 0 -1.1664>}]
 dyn_fr_r_hair_2 [{shapeType=Sphere, radius=1, shapeName=dyn_fr_r_hair_2_sphere, center=<-0.6723 -0.3642 0.9158>}]
 dyn_fr_r_hair_3 [{shapeType=Sphere, radius=1.5, shapeName=dyn_fr_r_hair_3_sphere, center=<-0.9709 -0.1036 1.2929>}]
 dyn_fr_strandA_2 [{shapeType=Sphere, radius=1, shapeName=dyn_fr_strandA_2_sphere, center=<-0.5141 1.8241 0.6663>}]
 dyn_fr_strandA_3 [{shapeType=Sphere, radius=1, shapeName=dyn_fr_strandA_3_sphere, center=<-0.4977 1.2294 0.4354>}]
 dyn_fr_strandA_4 [{shapeType=Sphere, radius=1, shapeName=dyn_fr_strandA_4_sphere, center=<-0.4865 1.1906 0.3654>}]
 dyn_fr_strandB_1 [{shapeType=Sphere, radius=1, shapeName=dyn_fr_strandB_1_sphere, center=<-1.1359 2.1154 1.6802>}]
 dyn_fr_strandB_2 [{shapeType=Sphere, radius=1, shapeName=dyn_fr_strandB_2_sphere, center=<-0.6507 1.1914 0.5069>}]
 dyn_fr_strandB_3 [{shapeType=Sphere, radius=1, shapeName=dyn_fr_strandB_3_sphere, center=<0 0.6698 0.1296>}]
 dyn_fr_strandC_1 [{shapeType=Sphere, radius=1, shapeName=dyn_fr_strandC_1_sphere, center=<0 0.1545 0.4932>}]
 dyn_fr_strandC_2 [{shapeType=Sphere, radius=1.25, shapeName=dyn_fr_strandC_2_sphere, center=<0 0.7784 0.198>}]
 dyn_fr_strandD_1 [{shapeType=Sphere, radius=1, shapeName=dyn_fr_strandD_1_sphere, center=<0 0 0.9383>}]
 dyn_fr_strandD_2 [{shapeType=Sphere, radius=1.25, shapeName=dyn_fr_strandD_2_sphere, center=<0 0.5152 0.843>}]
 dyn_fr_strandD_3 [{shapeType=Sphere, radius=1.5, shapeName=dyn_fr_strandD_3_sphere, center=<-0.444 0.175 0.2164>}]
 dyn_fr_strandE_1 [{shapeType=Sphere, radius=1, shapeName=dyn_fr_strandE_1_sphere}]
 dyn_fr_strandE_2 [{shapeType=Sphere, radius=1.25, shapeName=dyn_fr_strandE_2_sphere, center=<0 0.8898 0.0288>}]
 dyn_l_hair_1 [{shapeType=Sphere, radius=1, shapeName=dyn_l_hair_1_sphere, center=<0 0 -0.6655>}]
 dyn_l_hair_2 [{shapeType=Sphere, radius=1.25, shapeName=dyn_l_hair_2_sphere, center=<-1.0874 0 -1>}]
 dyn_r_hair_1 [{shapeType=Sphere, radius=1, shapeName=dyn_r_hair_1_sphere, center=<-1.2309 0 0.8002>}]
 dyn_r_hair_2 [{shapeType=Sphere, radius=1.5, shapeName=dyn_r_hair_2_sphere, center=<-1.6079 0 0.8507>}]
 head [{shapeType=Capsule, length=4.0523, radius=6.75, center=<5.2666 3.8032 -0.0024>, rotation=<rot 0 0 -90>}]
}
constraints [
 {bone1Name=dyn_r_hair_2, bone2Name=dyn_r_hair_1, swing1LimitDegrees=20, swing2LimitDegrees=20, twistLimitDegrees=10}
 {bone1Name=dyn_fr_r_hair_3, bone2Name=dyn_fr_r_hair_2, swing1LimitDegrees=20, swing2LimitDegrees=20, twistLimitDegrees=10}
 {bone1Name=dyn_fr_strandB_2, bone2Name=dyn_fr_strandB_1, swing1LimitDegrees=30, swing2LimitDegrees=30, twistLimitDegrees=15}
 {bone1Name=dyn_fr_strandB_3, bone2Name=dyn_fr_strandB_2, swing1LimitDegrees=30, swing2LimitDegrees=30, twistLimitDegrees=15}
 {bone1Name=dyn_fr_strandA_4, bone2Name=dyn_fr_strandA_3, swing1LimitDegrees=30, swing2LimitDegrees=30, twistLimitDegrees=15}
 {bone1Name=dyn_fr_strandA_3, bone2Name=dyn_fr_strandA_2, swing1LimitDegrees=30, swing2LimitDegrees=30, twistLimitDegrees=15}
 {bone1Name=dyn_fr_l_hair_2, bone2Name=dyn_fr_l_hair_1, swing1LimitDegrees=20, swing2LimitDegrees=20, twistLimitDegrees=10}
 {bone1Name=dyn_l_hair_2, bone2Name=dyn_l_hair_1, swing1LimitDegrees=20, swing2LimitDegrees=20, twistLimitDegrees=10}
 {bone1Name=dyn_fr_strandE_2, bone2Name=dyn_fr_strandE_1, swing1LimitDegrees=25, swing2LimitDegrees=25, twistLimitDegrees=15}
 {bone1Name=dyn_fr_strandD_2, bone2Name=dyn_fr_strandD_1, swing1LimitDegrees=20, swing2LimitDegrees=20, twistLimitDegrees=10}
 {bone1Name=dyn_fr_strandD_3, bone2Name=dyn_fr_strandD_2, swing1LimitDegrees=20, swing2LimitDegrees=20, twistLimitDegrees=10}
 {bone1Name=dyn_fr_strandC_1, bone2Name=head, swing1LimitDegrees=20, swing2LimitDegrees=20, twistLimitDegrees=10}
 {bone1Name=dyn_fr_strandC_2, bone2Name=dyn_fr_strandC_1, swing1LimitDegrees=20, swing2LimitDegrees=20, twistLimitDegrees=10}
]
